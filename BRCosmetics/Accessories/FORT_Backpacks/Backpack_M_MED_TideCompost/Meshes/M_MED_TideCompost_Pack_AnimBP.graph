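# Anim Blueprint for the TideCompost backpack that adjusts the pack bone's placement on different character bodies.
# On customization load, it reads gender and body type from a custom part, then each update derives body-combination flags and a spine-yaw-based BackpackTranslation.
# The AnimGraph chains additive Transform (Modify) Bone nodes on the input pose for pack offsets and a roll, mostly with bool-type alphas.
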AnimBlueprint /BRCosmetics/Accessories/FORT_Backpacks/Backpack_M_MED_TideCompost/Meshes/M_MED_TideCompost_Pack_AnimBP
target_skeleton=&/M_MED_TideCompost_Pack_Skeleton
graphs [
 AnimGraph {nodes [
   n1 = Input Pose()
   n2 = "Transform (Modify) Bone - Bone: pack"!(
    ComponentPose="Transform (Modify) Bone - Bone: pack"!(ComponentPose=Local To Component(LocalPose=@n1), Alpha:float, Translation="0,0,0") -> (Pose) {rotationMode=BMM_Additive, translationMode=BMM_Additive, boneToModify {boneName=pack}}
    bAlphaBoolEnabled:bool
    Translation="0,-2,2"
    Rotation="0,0,15"
   ) -> (Pose) {
    alphaInputType=Bool
    rotationMode=BMM_Additive
    translationMode=BMM_Additive
    alphaBoolBlend {blendInTime=0.3, blendOutTime=0.3}
    boneToModify {boneName=pack}
    rotation {roll=15}
    translation {y=-2, z=2}
   }
   n3 = "Transform (Modify) Bone - Bone: pack"!(
    ComponentPose="Transform (Modify) Bone - Bone: pack"(
     ComponentPose="Transform (Modify) Bone - Bone: pack"(ComponentPose=@n2, Translation="0,0.4,0") {alphaInputType=Bool, translationMode=BMM_Additive, boneToModify {boneName=pack}, translation {y=0.4}}
     Translation="0,0.5,0"
    ) {alphaInputType=Bool, translationMode=BMM_Additive, boneToModify {boneName=pack}, translation {y=0.5}}
    bAlphaBoolEnabled:bool
    Translation="0,0.2,0"
   ) -> (Pose) {alphaInputType=Bool, translationMode=BMM_Additive, boneToModify {boneName=pack}, translation {y=0.2}}
   Output Pose(Result=Component To Local(ComponentPose="Transform (Modify) Bone - Bone: pack"(ComponentPose=@n3, Translation="0,0.2,0") {alphaInputType=Bool, translationMode=BMM_Additive, boneToModify {boneName=pack}, translation {y=0.2}})) {layerGroup=None}
  ]}
 EventGraph {nodes [
   Event BlueprintUpdateAnimation()
   n1 = GetOwnerSkeletalMeshComponent()
   n2 = BreakRotator(InRot=BreakTransform(InTransform=GetDeltaTransformFromRefPose(self=@n1, BoneName=spine_05, BaseName=spine_01)).Rotation)
   > Set BackpackTranslation(BackpackTranslation:Vector=MakeVector(Y=MapRangeClamped(Value=@n2.Yaw, InRangeA=0, InRangeB=-30, OutRangeB=-3))) -> (Output_Get:Vector)
   > Set is F Med(is F Med:bool=AND Boolean(A=Get isMedium() -> (isMedium:bool), B=Get isFemale() -> (isFemale:bool))) -> (Output_Get:bool)
   > Set is M Lrg(is M Lrg:bool=AND Boolean(A=Get isLarge() -> (isLarge:bool), B=Get isMale() -> (isMale:bool))) -> (Output_Get:bool)
   > Set is F Lrg(is F Lrg:bool=AND Boolean(A=Get isLarge(), B=Get isFemale())) -> (Output_Get:bool)
   n3 = Event OnCharacterCustomizationLoaded()
   n4 = Get Array Item with EFort Custom Part Type(Array:Array of Custom Character Part Object References=@n3.CustomParts) -> (Output:Custom Character Part Object Reference)
   n5 = Get GenderPermitted!(self:Custom Character Part Object Reference=@n4) -> (GenderPermitted:EFortCustomGender Enum=Invalid)
   > Set isMale(isMale:bool="Equal (Enum)"(A:EFortCustomGender Enum=@n5, B:EFortCustomGender Enum=Male)) -> (Output_Get:bool)
   > Set isFemale(isFemale:bool="Equal (Enum)"(A=@n5, B=Female)) -> (Output_Get:bool)
   n6 = Get BodyTypesPermitted!(self:Custom Character Part Object Reference=@n4) -> (BodyTypesPermitted:EFortCustomBodyType Enum=NONE)
   > Set isLarge(isLarge:bool="Equal (Enum)"(A:EFortCustomBodyType Enum=@n6, B:EFortCustomBodyType Enum=Large)) -> (Output_Get:bool)
   > Set isMedium(isMedium:bool="Equal (Enum)"(A=@n6, B=Medium)) -> (Output_Get:bool)
   TryGetPawnOwner()
  ]}
]
variables [
 BackpackTranslation {type=Vector}
 is F Lrg {type=boolean}
 is F Med {type=boolean}
 is M Lrg {type=boolean}
 isFemale {type=boolean}
 isLarge {type=boolean}
 isMale {type=boolean}
 isMedium {type=boolean}
]
